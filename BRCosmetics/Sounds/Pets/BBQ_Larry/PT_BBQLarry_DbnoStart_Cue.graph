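# Sound cue for the BBQ Larry pet voice line that, judging by its naming, plays when the player enters the down-but-not-out state.
# A random node picks between two equally weighted dialogue waves, both using the Claptrap speaker voice with no targets.
# It uses very-close pet attenuation and the creature pets sound class.
# A concurrency override allows one instance per owner and blocks new plays while one is active.
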SoundCue /BRCosmetics/Sounds/Pets/BBQ_Larry/PT_BBQLarry_DbnoStart_Cue
root=Random_0
nodes [
 {
  id=DialoguePlayer_0
  properties {dialogueWaveParameter {
    dialogueWave=&/PET_BBQLarry_PlayerDBNO_020_Dialogue
    context {speaker=/BRCosmetics/Sounds/Other/Fort_Human_VO/DialogueVoices/Claptrap_DV, targets []}
   }}
 }
 {
  id=DialoguePlayer_1
  properties {dialogueWaveParameter {
    dialogueWave=&/PET_BBQLarry_PlayerDBNO_030_Dialogue
    context {speaker=&Claptrap_DV, targets []}
   }}
 }
 {id=Random_0, children [DialoguePlayer_0, DialoguePlayer_1], properties {weights [1,1]}}
]
properties {
 attenuationSettings=/BRCosmetics/Sounds/Other/Attenuation/Fort_Pet_BBQLarry_VeryClose_Att
 bOverrideConcurrency=true
 maxDistance=2000
 soundClassObject=&Fort_Creatures_Pets
 volumeMultiplier=0.896
 concurrencyOverrides {bLimitToOwner=true, maxCount=1, resolutionRule=PreventNew}
}
