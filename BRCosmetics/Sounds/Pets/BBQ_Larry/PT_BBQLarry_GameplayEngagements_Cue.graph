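# Sound cue for the BBQ Larry pet's gameplay engagement voice lines, spoken through the Claptrap dialogue voice.
# A short random delay feeds a quality-level switch: the first two levels pick randomly from twelve dialogue waves, the last two from a three-line subset.
# All picks are equally weighted.
# Very-close attenuation and an owner-limited, single-instance concurrency that prevents new plays keep lines from overlapping.
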SoundCue /BRCosmetics/Sounds/Pets/BBQ_Larry/PT_BBQLarry_GameplayEngagements_Cue
root=Delay_0
nodes [
 {id=Delay_0, children [QualityLevel_0], properties {delayMax=0.3, delayMin=0.2}}
 {
  id=DialoguePlayer_0
  properties {dialogueWaveParameter {
    dialogueWave=&/PET_BBQLarry_Gameplay_120_Dialogue
    context {speaker=/BRCosmetics/Sounds/Other/Fort_Human_VO/DialogueVoices/Claptrap_DV, targets []}
   }}
 }
 {
  id=DialoguePlayer_1
  properties {dialogueWaveParameter {
    dialogueWave=&/PET_BBQLarry_Gameplay_010_Dialogue
    context {speaker=&Claptrap_DV, targets []}
   }}
 }
 {
  id=DialoguePlayer_10
  properties {dialogueWaveParameter {
    dialogueWave=&/PET_BBQLarry_Gameplay_080_Dialogue
    context {speaker=&Claptrap_DV, targets []}
   }}
 }
 {
  id=DialoguePlayer_11
  properties {dialogueWaveParameter {
    dialogueWave=&/PET_BBQLarry_Gameplay_090_Dialogue
    context {speaker=&Claptrap_DV, targets []}
   }}
 }
 {
  id=DialoguePlayer_13
  properties {dialogueWaveParameter {
    dialogueWave=&/PET_BBQLarry_Gameplay_101_Dialogue
    context {speaker=&Claptrap_DV, targets []}
   }}
 }
 {
  id=DialoguePlayer_14
  properties {dialogueWaveParameter {
    dialogueWave=&/PET_BBQLarry_Gameplay_110_Dialogue
    context {speaker=&Claptrap_DV, targets []}
   }}
 }
 {
  id=DialoguePlayer_15
  properties {dialogueWaveParameter {
    dialogueWave=&PET_BBQLarry_Gameplay_010_Dialogue
    context {speaker=&Claptrap_DV, targets []}
   }}
 }
 {
  id=DialoguePlayer_16
  properties {dialogueWaveParameter {
    dialogueWave=&/PET_BBQLarry_Gameplay_070_Dialogue
    context {speaker=&Claptrap_DV, targets []}
   }}
 }
 {
  id=DialoguePlayer_17
  properties {dialogueWaveParameter {
    dialogueWave=&PET_BBQLarry_Gameplay_101_Dialogue
    context {speaker=&Claptrap_DV, targets []}
   }}
 }
 {
  id=DialoguePlayer_3
  properties {dialogueWaveParameter {
    dialogueWave=&/PET_BBQLarry_Gameplay_021_Dialogue
    context {speaker=&Claptrap_DV, targets []}
   }}
 }
 {
  id=DialoguePlayer_4
  properties {dialogueWaveParameter {
    dialogueWave=&/PET_BBQLarry_Gameplay_030_Dialogue
    context {speaker=&Claptrap_DV, targets []}
   }}
 }
 {
  id=DialoguePlayer_5
  properties {dialogueWaveParameter {
    dialogueWave=&/PET_BBQLarry_Gameplay_040_Dialogue
    context {speaker=&Claptrap_DV, targets []}
   }}
 }
 {
  id=DialoguePlayer_6
  properties {dialogueWaveParameter {
    dialogueWave=&/PET_BBQLarry_Gameplay_050_Dialogue
    context {speaker=&Claptrap_DV, targets []}
   }}
 }
 {
  id=DialoguePlayer_7
  properties {dialogueWaveParameter {
    dialogueWave=&/PET_BBQLarry_Gameplay_060_Dialogue
    context {speaker=&Claptrap_DV, targets []}
   }}
 }
 {
  id=DialoguePlayer_8
  properties {dialogueWaveParameter {
    dialogueWave=&PET_BBQLarry_Gameplay_070_Dialogue
    context {speaker=&Claptrap_DV, targets []}
   }}
 }
 {id=QualityLevel_0, children [Random_0,Random_0,Random_1,Random_1]}
 {
  id=Random_0
  children [
   DialoguePlayer_1
   DialoguePlayer_3
   DialoguePlayer_4
   DialoguePlayer_5
   DialoguePlayer_6
   DialoguePlayer_7
   DialoguePlayer_8
   DialoguePlayer_10
   DialoguePlayer_11
   DialoguePlayer_13
   DialoguePlayer_14
   DialoguePlayer_0
  ]
  properties {weights [1,1,1,1,1,1,1,1,1,1,1,1]}
 }
 {id=Random_1, children [DialoguePlayer_15, DialoguePlayer_16, DialoguePlayer_17], properties {weights [1,1,1]}}
]
properties {
 attenuationSettings=/BRCosmetics/Sounds/Other/Attenuation/Fort_Pet_BBQLarry_VeryClose_Att
 bOverrideConcurrency=true
 maxDistance=2000
 soundClassObject=&Fort_Creatures_Pets
 volumeMultiplier=0.896
 concurrencyOverrides {bLimitToOwner=true, maxCount=1, resolutionRule=PreventNew}
}
