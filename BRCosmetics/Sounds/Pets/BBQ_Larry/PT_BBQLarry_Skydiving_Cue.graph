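# Sound cue that plays skydiving voice lines and happy or terrified screams for the BBQ Larry pet, using the Claptrap dialogue voice.
# A quality switch sends the first two levels to a random pick of ten dialogue waves and the last two to a three-wave subset.
# Both branches add a short random delay before playback.
# Close-range pet attenuation applies, and concurrency allows one instance per owner, blocking new ones.
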SoundCue /BRCosmetics/Sounds/Pets/BBQ_Larry/PT_BBQLarry_Skydiving_Cue
root=QualityLevel_0
nodes [
 {id=Delay_0, children [Random_0], properties {delayMax=0.3, delayMin=0.2}}
 {id=Delay_1, children [Random_3], properties {delayMax=0.3, delayMin=0.2}}
 {
  id=DialoguePlayer_0
  properties {dialogueWaveParameter {
    dialogueWave=&/PET_BBQLarry_Skydiving_TerrifiedScreams_012_Dialogue
    context {speaker=/BRCosmetics/Sounds/Other/Fort_Human_VO/DialogueVoices/Claptrap_DV, targets []}
   }}
 }
 {
  id=DialoguePlayer_1
  properties {dialogueWaveParameter {
    dialogueWave=&/PET_BBQLarry_Skydiving_050_Dialogue
    context {speaker=&Claptrap_DV, targets []}
   }}
 }
 {
  id=DialoguePlayer_10
  properties {dialogueWaveParameter {
    dialogueWave=&/PET_BBQLarry_Skydiving_HappyScreams_012_Dialogue
    context {speaker=&Claptrap_DV, targets []}
   }}
 }
 {
  id=DialoguePlayer_11
  properties {dialogueWaveParameter {
    dialogueWave=&/PET_BBQLarry_Skydiving_TerrifiedScreams_010_Dialogue
    context {speaker=&Claptrap_DV, targets []}
   }}
 }
 {
  id=DialoguePlayer_12
  properties {dialogueWaveParameter {
    dialogueWave=&/PET_BBQLarry_Skydiving_TerrifiedScreams_011_Dialogue
    context {speaker=&Claptrap_DV, targets []}
   }}
 }
 {
  id=DialoguePlayer_14
  properties {dialogueWaveParameter {
    dialogueWave=&/PET_BBQLarry_Skydiving_HappyScreams_010_Dialogue
    context {speaker=&Claptrap_DV, targets []}
   }}
 }
 {
  id=DialoguePlayer_15
  properties {dialogueWaveParameter {
    dialogueWave=&PET_BBQLarry_Skydiving_TerrifiedScreams_011_Dialogue
    context {speaker=&Claptrap_DV, targets []}
   }}
 }
 {
  id=DialoguePlayer_4
  properties {dialogueWaveParameter {
    dialogueWave=&/PET_BBQLarry_Skydiving_040_Dialogue
    context {speaker=&Claptrap_DV, targets []}
   }}
 }
 {
  id=DialoguePlayer_5
  properties {dialogueWaveParameter {
    dialogueWave=&PET_BBQLarry_Skydiving_050_Dialogue
    context {speaker=&Claptrap_DV, targets []}
   }}
 }
 {
  id=DialoguePlayer_6
  properties {dialogueWaveParameter {
    dialogueWave=&/PET_BBQLarry_Skydiving_060_Dialogue
    context {speaker=&Claptrap_DV, targets []}
   }}
 }
 {
  id=DialoguePlayer_7
  properties {dialogueWaveParameter {
    dialogueWave=&/PET_BBQLarry_Skydiving_070_Dialogue
    context {speaker=&Claptrap_DV, targets []}
   }}
 }
 {
  id=DialoguePlayer_8
  properties {dialogueWaveParameter {
    dialogueWave=&PET_BBQLarry_Skydiving_HappyScreams_010_Dialogue
    context {speaker=&Claptrap_DV, targets []}
   }}
 }
 {
  id=DialoguePlayer_9
  properties {dialogueWaveParameter {
    dialogueWave=&/PET_BBQLarry_Skydiving_HappyScreams_011_Dialogue
    context {speaker=&Claptrap_DV, targets []}
   }}
 }
 {id=QualityLevel_0, children [Delay_0,Delay_0,Delay_1,Delay_1]}
 {
  id=Random_0
  children [
   DialoguePlayer_4
   DialoguePlayer_5
   DialoguePlayer_6
   DialoguePlayer_7
   DialoguePlayer_8
   DialoguePlayer_9
   DialoguePlayer_10
   DialoguePlayer_11
   DialoguePlayer_12
   DialoguePlayer_0
  ]
  properties {weights [1,1,1,1,1,1,1,1,1,1]}
 }
 {
  id=Random_3
  children [DialoguePlayer_1, DialoguePlayer_14, DialoguePlayer_15]
  properties {bShouldExcludeFromBranchCulling=true, weights [1,1,1]}
 }
]
properties {
 attenuationSettings=/BRCosmetics/Sounds/Other/Attenuation/Fort_Pet_BBQLarry_VeryClose_Att
 bOverrideConcurrency=true
 maxDistance=2000
 soundClassObject=&Fort_Creatures_Pets
 volumeMultiplier=0.896
 concurrencyOverrides {bLimitToOwner=true, maxCount=1, resolutionRule=PreventNew}
}
